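# Niagara ribbon swing trail for the DoughDisk pickaxe, built on Fortnite's AnimTrail modules that follow the weapon's skeletal mesh.
# BodyTrail001 draws a short body ribbon that narrows over particle life and fades with emitter age.
# TipTrail001 adds a camera-offset tip ribbon with time-scrolled U, two renderers and a dark, slightly wider secondary layer.
# Time-of-day inputs adjust trail opacity and tint; the older TipTrail emitter is disabled.
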NiagaraSystem /BRCosmetics/Weapons/FORT_Melee/Pickaxe_DoughDisk/FX/NS_Pickaxe_DoughDisk_Trail
dependencies {
 usedDataInterfaces [&NiagaraDataInterfaceCurve, &NiagaraDataInterfaceColorCurve]
 usedDynamicInputs []
 usedModules [
  &self:SystemSpawnScript.SystemScriptSource.SystemScriptGraph.NiagaraNodeAssignment_0.SetVariables_#1
  &SystemState
 ]
}
emitters {
 BodyTrail001 {
  emitterUpdateScript [
   &EmitterState()
   &FN_SpawnAnimTrail_v2()
  ]
  particleSpawnScript [
   /Game/Effects/Niagara/Modules/AnimTrail/FN_InitializeAnimTrail/FN_InitializeAnimTrailBoneIndecies(Second Bone or Socket Index=1)
   &FN_InitializeAnimTrail(AnimTrailWidth=2, Color=<rgba 0 0 0 4>, Lifetime=0.23, SkeletalMesh=$System.Skeletal Mesh) {module=FN_InitializeAnimTrail001}
  ]
  particleUpdateScript [
   &FN_UpdateAnimTrail(SkeletalMesh=$System.Skeletal Mesh)
   &ScaleRibbonWidth(Ribbon Width Scale=&FloatFromCurve(FloatCurve=DataInterface() {propertyValues {Curve {keys [
         {arriveTangent=-1.2579, interpMode=RCIM_Cubic, time=0, value=1}
         {arriveTangent=-0.9664, interpMode=RCIM_Cubic, leaveTangent=-0.9664, tangentMode=RCTM_User, time=1, value=0}
        ]}}}))
   &ScaleColor(
    Scale Alpha=&Multiply_Float(
     A=&FloatFromCurve(
      FloatCurve=DataInterface() {propertyValues {Curve {keys [
          {interpMode=RCIM_Cubic, tangentMode=RCTM_User, time=0, value=1}
          {arriveTangent=-2.1778, interpMode=RCIM_Cubic, leaveTangent=-2.1778, tangentMode=RCTM_User, time=1, value=0}
         ]}}}
      CurveIndex=&RemapRange(Input Value To Remap=$Emitter.Age, Input Min=0, Input Max=0.23, Output Min=0, Output Max=1)
     )
     B=&FloatFromCurve(
      FloatCurve=DataInterface() {propertyValues {Curve {keys [
          {interpMode=RCIM_Cubic, tangentMode=RCTM_User, time=0, value=0.5}
          {interpMode=RCIM_Cubic, tangentMode=RCTM_User, time=0.5, value=1}
          {tangentMode=RCTM_User, time=1, value=0.5}
         ]}}}
      CurveIndex=Linked() {linkedVariable {name=NPC.FortniteNPC.TimeOfDay_0-1, type=NiagaraFloat}}
     )
    )
    Scale RGB=&VectorFromFloat(Value=&FNTimeOfDayScalar(Night Scale=0.5, Morning Scale=1, Evening Scale=1))
    ScaleA=true
    ScaleRGB=false
   ) {module=ScaleColor002}
  ]
  rendererClasses [&NiagaraRibbonRendererProperties]
  renderers [{rendererClass=&NiagaraRibbonRendererProperties}]
 }
 TipTrail {
  bEnabled=false
  emitterUpdateScript [
   &EmitterState()
   &FN_SpawnAnimTrail(Spawn Interval=0)
  ]
  particleSpawnScript [
   /Game/Effects/Niagara/Modules/AnimTrail/FN_InitializeAnimTrailBoneIndecies_v2/FN_InitializeAnimTrailBoneIndecies()
   &FN_InitializeAnimTrail(Color=<rgba 0.6419 0.6419 0.6419 1>, Lifetime=0.2, SkeletalMesh=$System.Skeletal Mesh) {module=FN_InitializeAnimTrail001}
  ]
  particleUpdateScript [
   &FN_UpdateAnimTrail(SkeletalMesh=$System.Skeletal Mesh)
   &ScaleColor(
    Scale Alpha=&FloatFromCurve(
     FloatCurve=DataInterface() {propertyValues {Curve {keys [
         {interpMode=RCIM_Cubic, tangentMode=RCTM_User, time=0, value=2}
         {arriveTangent=-5.5946, interpMode=RCIM_Cubic, leaveTangent=-5.5946, tangentMode=RCTM_User, time=1, value=0}
        ]}}}
     CurveIndex=&RemapRange(Input Value To Remap=$Emitter.Age, Input Min=0, Input Max=0.2, Output Min=0, Output Max=1)
    )
    Scale RGB=&Multiply_Vector(
     A=&MakeVectorFromLinearColorRGB(Color=&ColorFromCurve(ColorCurve=DataInterface() {propertyValues {
         AlphaCurve {keys [<at 0.3581 1 RCIM_Cubic>,<at 1 0 RCIM_Cubic>]}
         BlueCurve {keys [<at 0 3 RCIM_Cubic>,<at 0.1591 1>,<at 1 0.2593 RCIM_Cubic>]}
         GreenCurve {keys [<at 0 3 RCIM_Cubic>,<at 0.1591 1>,<at 1 0.2593 RCIM_Cubic>]}
         RedCurve {keys [<at 0 3 RCIM_Cubic>,<at 0.1591 1>,<at 1 0.2593 RCIM_Cubic>]}
        }}))
     B=&VectorFromFloat(Value=&FNTimeOfDayScalar(Night Scale=0.3, Morning Scale=0.7, Day Scale=0.5, Evening Scale=0.2))
    )
    ScaleA=true
    ScaleRGB=true
   )
   &ParticleState()
   /SetVariables(Particles.RibbonU0Override=&Multiply_Float(A=$Engine.Time, B=4)) {isSetParameters=true, module=SetVariables_#2}
   &CameraOffset(Camera Offset Amount=10)
   &ScaleRibbonWidth(Ribbon Width Scale=&FloatFromCurve(FloatCurve=DataInterface() {propertyValues {Curve {keys [
         {arriveTangent=-2.083, interpMode=RCIM_Cubic, leaveTangent=-2.083, tangentMode=RCTM_User, time=0.25, value=1}
         {tangentMode=RCTM_User, time=1, value=0.25}
        ]}}}))
   /SetVariables(
    Particles.SecondaryColor:LinearColor=&MultiplyLinearColor(
     Linear Color=<rgba 0.02 0.02 0.02 0.8>
     Scale Factor=&MakeVector4(
      X=1
      Y=1
      Z=1
      W=&Multiply_Float(
       A=&FloatFromCurve(
        FloatCurve=DataInterface() {propertyValues {Curve {keys [
            {interpMode=RCIM_Cubic, tangentMode=RCTM_User, time=0, value=2}
            {arriveTangent=-5.5946, interpMode=RCIM_Cubic, leaveTangent=-5.5946, tangentMode=RCTM_User, time=1, value=0}
           ]}}}
        CurveIndex=&RemapRange(Input Value To Remap=$Emitter.Age, Input Min=0, Input Max=0.2, Output Min=0, Output Max=1)
       )
       B=&FloatFromCurve(
        FloatCurve=DataInterface() {propertyValues {Curve {keys [
            {interpMode=RCIM_Cubic, tangentMode=RCTM_User, time=0, value=0.5}
            {interpMode=RCIM_Cubic, tangentMode=RCTM_User, time=0.5, value=1}
            {tangentMode=RCTM_User, time=1, value=0.5}
           ]}}}
        CurveIndex=Linked() {linkedVariable {name=NPC.FortniteNPC.TimeOfDay_0-1, type=NiagaraFloat}}
       )
      )
     )
    )
    Particles.SecondaryWidth:NiagaraFloat=&Multiply_Float(A=$Particles.RibbonWidth, B=1.1)
   ) {isSetParameters=true, module=SetVariables_#3}
  ]
  rendererClasses [&NiagaraRibbonRendererProperties]
  renderers [
   {rendererClass=&NiagaraRibbonRendererProperties}
   {rendererClass=&NiagaraRibbonRendererProperties, rendererIndex=1}
  ]
 }
 TipTrail001 {
  emitterUpdateScript [
   &EmitterState()
   &FN_SpawnAnimTrail(Spawn Interval=0)
  ]
  particleSpawnScript [
   &FN_InitializeAnimTrailBoneIndecies()
   &FN_InitializeAnimTrail(Color=<rgba 0.6419 0.6419 0.6419 1>, Lifetime=0.2, SkeletalMesh=$System.Skeletal Mesh) {module=FN_InitializeAnimTrail001}
  ]
  particleUpdateScript [
   &FN_UpdateAnimTrail(SkeletalMesh=$System.Skeletal Mesh)
   &ScaleColor(
    Scale Alpha=&FloatFromCurve(
     FloatCurve=DataInterface() {propertyValues {Curve {keys [
         {interpMode=RCIM_Cubic, tangentMode=RCTM_User, time=0, value=2}
         {arriveTangent=-5.5946, interpMode=RCIM_Cubic, leaveTangent=-5.5946, tangentMode=RCTM_User, time=1, value=0}
        ]}}}
     CurveIndex=&RemapRange(Input Value To Remap=$Emitter.Age, Input Min=0, Input Max=0.2, Output Min=0, Output Max=1)
    )
    Scale RGB=&MakeVectorFromLinearColorRGB(Color=&FNTimeOfDayColor(Night Color=<rgba 1 1 1 1>, Morning Color=<rgba 1 1 1 1>, Day Color=<rgba 1 1 1 1>, Evening Color=<rgba 1 1 1 1>))
    ScaleA=true
    ScaleRGB=true
   ) {enabled=false}
   &ScaleColor(
    Scale Alpha=&FloatFromCurve(
     FloatCurve=DataInterface() {propertyValues {Curve {keys [
         {interpMode=RCIM_Cubic, tangentMode=RCTM_User, time=0, value=2}
         {arriveTangent=-5.5946, interpMode=RCIM_Cubic, leaveTangent=-5.5946, tangentMode=RCTM_User, time=1, value=0}
        ]}}}
     CurveIndex=&RemapRange(Input Value To Remap=$Emitter.Age, Input Min=0, Input Max=0.2, Output Min=0, Output Max=1)
    )
    Scale RGB=&Multiply_Vector(
     A=&MakeVectorFromLinearColorRGB(Color=&ColorFromCurve(ColorCurve=DataInterface() {propertyValues {
         AlphaCurve {keys [<at 0.3581 1 RCIM_Cubic>,<at 1 0 RCIM_Cubic>]}
         BlueCurve {keys [<at 0 3 RCIM_Cubic>,<at 0.1591 1>,<at 1 0.2593 RCIM_Cubic>]}
         GreenCurve {keys [<at 0 3 RCIM_Cubic>,<at 0.1591 1>,<at 1 0.2593 RCIM_Cubic>]}
         RedCurve {keys [<at 0 3 RCIM_Cubic>,<at 0.1591 1>,<at 1 0.2593 RCIM_Cubic>]}
        }}))
     B=&MakeVectorFromLinearColorRGB(Color=&FNTimeOfDayColor(
       Night Color=<rgba 0.111 0.2442 0.5549 1>
       Morning Color=<rgba 1 0.85 0.7 1>
       Day Color=<rgba 0.2987 0.3212 0.3686 1>
       Evening Color=<rgba 0.6071 0.3206 0.1118 1>
      ))
    )
    ScaleA=true
    ScaleRGB=true
   ) {module=ScaleColor001}
   &ParticleState()
   /SetVariables(Particles.RibbonU0Override=&Multiply_Float(A=$Engine.Time, B=4)) {isSetParameters=true, module=SetVariables_#2}
   &CameraOffset(Camera Offset Amount=10)
   &ScaleRibbonWidth(Ribbon Width Scale=&FloatFromCurve(FloatCurve=DataInterface() {propertyValues {Curve {keys [
         {arriveTangent=-2.083, interpMode=RCIM_Cubic, leaveTangent=-2.083, tangentMode=RCTM_User, time=0.25, value=1}
         {tangentMode=RCTM_User, time=1, value=0.25}
        ]}}}))
   /SetVariables(
    Particles.SecondaryColor:LinearColor=&MultiplyLinearColor(
     Linear Color=<rgba 0.02 0.02 0.02 0.8>
     Scale Factor=&MakeVector4(
      X=1
      Y=1
      Z=1
      W=&Multiply_Float(
       A=&FloatFromCurve(
        FloatCurve=DataInterface() {propertyValues {Curve {keys [
            {interpMode=RCIM_Cubic, tangentMode=RCTM_User, time=0, value=2}
            {arriveTangent=-5.5946, interpMode=RCIM_Cubic, leaveTangent=-5.5946, tangentMode=RCTM_User, time=1, value=0}
           ]}}}
        CurveIndex=&RemapRange(Input Value To Remap=$Emitter.Age, Input Min=0, Input Max=0.2, Output Min=0, Output Max=1)
       )
       B=&FloatFromCurve(
        FloatCurve=DataInterface() {propertyValues {Curve {keys [
            {interpMode=RCIM_Cubic, tangentMode=RCTM_User, time=0, value=0.5}
            {interpMode=RCIM_Cubic, tangentMode=RCTM_User, time=0.5, value=1}
            {tangentMode=RCTM_User, time=1, value=0.5}
           ]}}}
        CurveIndex=Linked() {linkedVariable {name=NPC.FortniteNPC.TimeOfDay_0-1, type=NiagaraFloat}}
       )
      )
     )
    )
    Particles.SecondaryWidth:NiagaraFloat=&Multiply_Float(A=$Particles.RibbonWidth, B=1.1)
   ) {isSetParameters=true, module=SetVariables_#3}
  ]
  rendererClasses [&NiagaraRibbonRendererProperties]
  renderers [
   {rendererClass=&NiagaraRibbonRendererProperties}
   {rendererClass=&NiagaraRibbonRendererProperties, rendererIndex=1}
  ]
 }
}
summary {
 systemName=NS_Pickaxe_DoughDisk_Trail
 emitters [
  {emitterName=BodyTrail001, rendererClasses [&NiagaraRibbonRendererProperties]}
  {bEnabled=false, emitterName=TipTrail, rendererClasses [&NiagaraRibbonRendererProperties]}
  {emitterName=TipTrail001, rendererClasses [&NiagaraRibbonRendererProperties]}
 ]
 userVariables [
  User.RibbonIDOffset {type=NiagaraInt32}
  User.SkeletalMesh {
   type=NiagaraDataInterfaceSkeletalMesh
   defaultValue {
    struct=NiagaraExt_VariableValue_DataInterface
    value {dataInterface=&self:NiagaraDataInterfaceSkeletalMesh_0}
   }
  }
 ]
}
