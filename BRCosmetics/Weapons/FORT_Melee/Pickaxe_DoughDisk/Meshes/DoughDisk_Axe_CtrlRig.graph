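# Control Rig for the DoughDisk pickaxe that procedurally wiggles its three tentacle bone chains.
# On execution it branches on the Is_R_handCR bool, false by default, to one of two tunings of the same chains.
# Each branch runs ChainHarmonicsPerItem nodes in sequence on the tentacle 06, 04 and 02 base bones.
# Motion comes only from wave harmonics with per-chain speeds, offsets and curves, as pendulum and reach are disabled.
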ControlRigBlueprint /BRCosmetics/Weapons/FORT_Melee/Pickaxe_DoughDisk/Meshes/DoughDisk_Axe_CtrlRig
graphs [
 RigVMFunctionLibrary {links [], nodes []}
 RigVMModel {nodes [
   RigUnit_BeginExecution = &RigUnit_BeginExecution()
   RigVMFunction_ControlFlowBranch = &RigVMFunction_ControlFlowBranch(ExecuteContext=@RigUnit_BeginExecution.ExecutePin, Condition=RigVMVariableNode() {variable=Is_R_handCR}.Value)
   /Script/ControlRig.RigUnit_ChainHarmonicsPerItem(
    ChainRoot=(Type=Bone,Name="Ppicaxe_tentkl_06_l_base")
    DrawWorldOffset=(Rotation=(X=0,Y=0,Z=0,W=1),Translation=(X=0,Y=0,Z=0),Scale3D=(X=1,Y=1,Z=1))
    Pendulum=(bEnabled=False,PendulumStiffness=2,PendulumGravity=(X=0,Y=0,Z=0),PendulumBlend=0.75,PendulumDrag=0.98,PendulumMinimum=0,PendulumMaximum=0.1,PendulumEase=Linear,UnwindAxis=(X=0,Y=1,Z=0),UnwindMinimum=0.2,UnwindMaximum=0.05)
    Reach=(bEnabled=False,ReachTarget=(X=0,Y=0,Z=0),ReachAxis=(X=1,Y=0,Z=0),ReachMinimum=0,ReachMaximum=0,ReachEase=Linear)
    Speed=(X=1.2,Y=1.4,Z=1.2)
    Wave=(bEnabled=True,WaveFrequency=(X=1,Y=0.8,Z=1),WaveAmplitude=(X=0.4,Y=0.8,Z=1),WaveOffset=(X=0,Y=1,Z=0.5),WaveNoise=(X=0,Y=0,Z=-0.1),WaveMinimum=0,WaveMaximum=1,WaveEase=Linear)
    WaveCurve=(EditorCurveData=(Keys=((),(Time=0.3588,Value=0.288))))
    bDrawDebug=false
    ExecutePin=@RigVMFunction_ControlFlowBranch.False
   ) {is=struct RigVMUnitNode}
   > &RigUnit_ChainHarmonicsPerItem(
    ChainRoot=(Type=Bone,Name="Ppicaxe_tentkl_04_l_base")
    DrawWorldOffset=(Rotation=(X=0,Y=0,Z=0,W=1),Translation=(X=0,Y=0,Z=0),Scale3D=(X=1,Y=1,Z=1))
    Pendulum=(bEnabled=False,PendulumStiffness=2,PendulumGravity=(X=0,Y=0,Z=0),PendulumBlend=0.75,PendulumDrag=0.98,PendulumMinimum=0,PendulumMaximum=0.1,PendulumEase=Linear,UnwindAxis=(X=0,Y=1,Z=0),UnwindMinimum=0.2,UnwindMaximum=0.05)
    Reach=(bEnabled=False,ReachTarget=(X=0,Y=0,Z=0),ReachAxis=(X=1,Y=0,Z=0),ReachMinimum=0,ReachMaximum=0,ReachEase=Linear)
    Speed=(X=1.4,Y=1.7,Z=1.4)
    Wave=(bEnabled=true,WaveFrequency=(X=1,Y=0.8,Z=-3),WaveAmplitude=(X=0.4,Y=0.8,Z=0.5),WaveOffset=(X=0,Y=1,Z=1),WaveNoise=(X=0,Y=0,Z=0),WaveMinimum=0,WaveMaximum=1,WaveEase=Linear)
    WaveCurve=(EditorCurveData=(Keys=((),(Time=0.1764,Value=0.4831))))
    bDrawDebug=false
   )
   > &RigUnit_ChainHarmonicsPerItem(
    ChainRoot=(Type=Bone,Name="Ppicaxe_tentkl_02_l_base")
    DrawWorldOffset=(Rotation=(X=0,Y=0,Z=0,W=1),Translation=(X=0,Y=0,Z=0),Scale3D=(X=1,Y=1,Z=1))
    Pendulum=(bEnabled=False,PendulumStiffness=2,PendulumGravity=(X=0,Y=0,Z=0),PendulumBlend=0.75,PendulumDrag=0.98,PendulumMinimum=0,PendulumMaximum=0.1,PendulumEase=Linear,UnwindAxis=(X=0,Y=1,Z=0),UnwindMinimum=0.2,UnwindMaximum=0.05)
    Reach=(bEnabled=False,ReachTarget=(X=0,Y=0,Z=0),ReachAxis=(X=1,Y=0,Z=0),ReachMinimum=0,ReachMaximum=0,ReachEase=Linear)
    Speed=(X=1.2,Y=1.2,Z=1.4)
    Wave=(bEnabled=True,WaveFrequency=(X=1,Y=0.8,Z=-3),WaveAmplitude=(X=0.4,Y=0.8,Z=0.5),WaveOffset=(X=0,Y=1,Z=1),WaveNoise=(X=0,Y=0,Z=0),WaveMinimum=0,WaveMaximum=1,WaveEase=Linear)
    WaveCurve=(EditorCurveData=(Keys=((),(InterpMode=RCIM_Constant,Time=0.5685,Value=0.5127))))
    bDrawDebug=false
   )
   &RigUnit_ChainHarmonicsPerItem(
    ChainRoot=(Type=Bone,Name="Ppicaxe_tentkl_06_l_base")
    DrawWorldOffset=(Rotation=(X=0,Y=0,Z=0,W=1),Translation=(X=0,Y=0,Z=0),Scale3D=(X=1,Y=1,Z=1))
    Pendulum=(bEnabled=False,PendulumStiffness=2,PendulumGravity=(X=0,Y=0,Z=0),PendulumBlend=0.75,PendulumDrag=0.98,PendulumMinimum=0,PendulumMaximum=0.1,PendulumEase=Linear,UnwindAxis=(X=0,Y=1,Z=0),UnwindMinimum=0.2,UnwindMaximum=0.05)
    Reach=(bEnabled=False,ReachTarget=(X=0,Y=0,Z=0),ReachAxis=(X=1,Y=0,Z=0),ReachMinimum=0,ReachMaximum=0,ReachEase=Linear)
    Speed=(X=1.7,Y=1.7,Z=1.6)
    Wave=(bEnabled=True,WaveFrequency=(X=1,Y=0.8,Z=1),WaveAmplitude=(X=0.7,Y=0.8,Z=1),WaveOffset=(X=0,Y=0.9,Z=0.5),WaveNoise=(X=0,Y=0.2,Z=0.1),WaveMinimum=0,WaveMaximum=1,WaveEase=Linear)
    WaveCurve=(EditorCurveData=(Keys=((),(Time=0.3744,Value=0.3066))))
    bDrawDebug=false
    ExecutePin=@RigVMFunction_ControlFlowBranch.True
   )
   > &RigUnit_ChainHarmonicsPerItem(
    ChainRoot=(Type=Bone,Name="Ppicaxe_tentkl_04_l_base")
    DrawWorldOffset=(Rotation=(X=0,Y=0,Z=0,W=1),Translation=(X=0,Y=0,Z=0),Scale3D=(X=1,Y=1,Z=1))
    Pendulum=(bEnabled=False,PendulumStiffness=2,PendulumGravity=(X=0,Y=0,Z=0),PendulumBlend=0.75,PendulumDrag=0.98,PendulumMinimum=0,PendulumMaximum=0.1,PendulumEase=Linear,UnwindAxis=(X=0,Y=1,Z=0),UnwindMinimum=0.2,UnwindMaximum=0.05)
    Reach=(bEnabled=False,ReachTarget=(X=0,Y=0,Z=0),ReachAxis=(X=1,Y=0,Z=0),ReachMinimum=0,ReachMaximum=0,ReachEase=Linear)
    Speed=(X=1.8,Y=1.3,Z=1.6)
    Wave=(bEnabled=true,WaveFrequency=(X=1,Y=0.8,Z=-3),WaveAmplitude=(X=0.4,Y=0.8,Z=0.5),WaveOffset=(X=0,Y=0.7,Z=1),WaveNoise=(X=0,Y=0.2,Z=0),WaveMinimum=0,WaveMaximum=1,WaveEase=Linear)
    WaveCurve=(EditorCurveData=(Keys=((),(Time=0.1764,Value=0.4831))))
    bDrawDebug=false
   )
   > &RigUnit_ChainHarmonicsPerItem(
    ChainRoot=(Type=Bone,Name="Ppicaxe_tentkl_02_l_base")
    DrawWorldOffset=(Rotation=(X=0,Y=0,Z=0,W=1),Translation=(X=0,Y=0,Z=0),Scale3D=(X=1,Y=1,Z=1))
    Pendulum=(bEnabled=False,PendulumStiffness=2,PendulumGravity=(X=0,Y=0,Z=0),PendulumBlend=0.75,PendulumDrag=0.98,PendulumMinimum=0,PendulumMaximum=0.1,PendulumEase=Linear,UnwindAxis=(X=0,Y=1,Z=0),UnwindMinimum=0.2,UnwindMaximum=0.05)
    Reach=(bEnabled=False,ReachTarget=(X=0,Y=0,Z=0),ReachAxis=(X=1,Y=0,Z=0),ReachMinimum=0,ReachMaximum=0,ReachEase=Linear)
    Speed=(X=1.7,Y=1.8,Z=1.2)
    Wave=(bEnabled=True,WaveFrequency=(X=1,Y=0.8,Z=-3),WaveAmplitude=(X=0.4,Y=0.6,Z=0.8),WaveOffset=(X=0,Y=-0.3,Z=1),WaveNoise=(X=0,Y=0.1,Z=0),WaveMinimum=0,WaveMaximum=1,WaveEase=Linear)
    WaveCurve=(EditorCurveData=(Keys=((),(InterpMode=RCIM_Constant,Time=0.4391,Value=0.5127))))
    bDrawDebug=false
   )
  ]}
]
hierarchy [
 b_AR_Root {loc=<0 0 0>, rot=<0 0 0 1>}
 Melee_TwoHanded_Handle {parent=b_AR_Root, loc=<0 0 0>, rot=<0 0 0.7071 -0.7071>}
 Ppicaxe_02_l {parent=Melee_TwoHanded_Handle, loc=<-0.5554 10.9598 36.631>, rot=<0.1642 -0.1642 -0.6878 -0.6878>}
 Ppicaxe_tentkl_06_l_base {parent=Ppicaxe_02_l, loc=<4.5731 -0.0356 7.9938>, rot=<0.9921 -0.0045 -0.1195 0.0377>}
 Ppicaxe_tentkl_06_l {parent=Ppicaxe_tentkl_06_l_base, loc=<0 0 0>, rot=<0 0 0 1>}
 Ppicaxe_tentkl_05_l {parent=Ppicaxe_tentkl_06_l, loc=<4.646 0.1562 0.0119>, rot=<0 0 0 1>}
 Ppicaxe_tentkl_04_l_base {parent=Ppicaxe_02_l, loc=<5.7684 -0.0356 2.6884>, rot=<0.9665 -0.0096 -0.2538 0.0367>}
 Ppicaxe_tentkl_04_l {parent=Ppicaxe_tentkl_04_l_base, loc=<0 0 0>, rot=<0 0 0 1>}
 Ppicaxe_tentkl_03_l {parent=Ppicaxe_tentkl_04_l, loc=<5.1785 -0.0374 -0.2965>, rot=<0 -0.051 0 -0.9987>}
 Ppicaxe_tentkl_02_l_base {parent=Ppicaxe_02_l, loc=<5.9178 -0.0356 -4.8331>, rot=<0.9572 -0.0109 -0.2871 0.0364>}
 Ppicaxe_tentkl_02_l {parent=Ppicaxe_tentkl_02_l_base, loc=<0 0 0>, rot=<0 0 0 1>}
 Ppicaxe_tentkl_01_l {parent=Ppicaxe_tentkl_02_l, loc=<3.6643 -0.1586 0.6562>, rot=<0 0.1733 0 -0.9849>}
]
variables [Is_R_handCR {type=bool, default=False}]
